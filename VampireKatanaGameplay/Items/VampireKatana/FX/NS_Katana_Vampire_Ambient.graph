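# Niagara system for the vampire katana's ambient effect: glowing ribbon trails and dark smoke that follow the weapon's skeletal mesh.
# AnimTrail and shorter-lived AnimTrailLower use FN anim-trail modules on separate system mesh variables, with waveform-blended velocity, fading from hot orange to dark red.
# Trail alpha passes through a time-of-day scalar.
# Particle_Smoke re-samples the user mesh each update, using random curves to vary SubUV sprites; a DEBUG sprite emitter is disabled.
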NiagaraSystem /VampireKatanaGameplay/Items/VampireKatana/FX/NS_Katana_Vampire_Ambient
dependencies {
 usedDataInterfaces [&NiagaraDataInterfaceCurve]
 usedDynamicInputs []
 usedModules [
  &self:SystemSpawnScript.SystemScriptSource.SystemScriptGraph.NiagaraNodeAssignment_0.SetVariables_#1
  &self:SystemSpawnScript.SystemScriptSource.SystemScriptGraph.NiagaraNodeAssignment_1.SetVariables_#2
  &SystemState
  &self:SystemSpawnScript.SystemScriptSource.SystemScriptGraph.NiagaraNodeAssignment_2.SetVariables_#3
 ]
}
emitters {
 AnimTrail {
  emitterUpdateScript [
   &EmitterState()
   &FN_SpawnAnimTrail()
  ]
  particleSpawnScript [
   &FN_InitializeAnimTrailBoneIndecies()
   &FN_InitializeAnimTrail(AnimTrailWidth=1.2, Lifetime=0.7, SkeletalMesh=$System.SkeletalMeshKatana)
   &AddVelocity(Velocity=&Lerp_Vector(
     A=$System.RandVecMin
     B=$System.RandVecMax
     Alpha=&Waveform("Waveform [1]"=::NewEnumerator12, "[1] Frequency"=1.25, "[1] Amplitude Min/Max"=<0 1>)
    ))
  ]
  particleUpdateScript [
   &ParticleState()
   &SolveForcesAndVelocity(Clamp Velocity=false, Limit Acceleration=false)
   &FN_UpdateAnimTrail(SkeletalMesh=$System.SkeletalMeshKatana)
   &ScaleColor(
    Scale Alpha=&FloatFromCurve(
     FloatCurve=DataInterface() {propertyValues {Curve {keys [{interpMode=RCIM_Cubic, tangentMode=RCTM_User, time=0, value=1}, {interpMode=RCIM_Cubic, tangentMode=RCTM_User, time=1, value=0}]}}}
     Scale Curve=&FNTimeOfDayScalar(Night Scale=1, Morning Scale=1, Evening Scale=1)
    )
    Scale RGB=&MakeVectorFromLinearColorRGB(Color=&LerpLinearColors(
      StartColor=<rgba 5 0.6247 0 1>
      EndColor=<rgba 0.2865 0 0 1>
      LerpFactor=&FloatFromCurve(FloatCurve=DataInterface() {propertyValues {Curve {keys [
           {arriveTangent=19.3856, interpMode=RCIM_Cubic, leaveTangent=19.3856, tangentMode=RCTM_User, time=0, value=0}
           {arriveTangent=1.3591, interpMode=RCIM_Cubic, leaveTangent=1.3591, tangentMode=RCTM_User, time=0.1, value=0.9}
           {interpMode=RCIM_Cubic, tangentMode=RCTM_User, time=0.4, value=1}
          ]}}})
     ))
    ScaleA=true
    ScaleRGB=true
   )
  ]
  rendererClasses [&NiagaraRibbonRendererProperties]
  renderers [{rendererClass=&NiagaraRibbonRendererProperties}]
 }
 AnimTrailLower {
  emitterUpdateScript [
   &EmitterState()
   &FN_SpawnAnimTrail()
  ]
  particleSpawnScript [
   &FN_InitializeAnimTrailBoneIndecies()
   &FN_InitializeAnimTrail(AnimTrailWidth=1.2, Lifetime=0.4, SkeletalMesh=$System.SkeletalMeshKatanaLow)
   &AddVelocity(Velocity=&Lerp_Vector(
     A=$System.RandVecMin
     B=$System.RandVecMax
     Alpha=&Waveform("Waveform [1]"=::NewEnumerator12, "[1] Frequency"=1.25, "[1] Amplitude Min/Max"=<0 1>)
    ))
  ]
  particleUpdateScript [
   &ParticleState()
   &SolveForcesAndVelocity(Clamp Velocity=false, Limit Acceleration=false)
   &FN_UpdateAnimTrail(SkeletalMesh=$System.SkeletalMeshKatanaLow)
   &ScaleColor(
    Scale Alpha=&FloatFromCurve(
     FloatCurve=DataInterface() {propertyValues {Curve {keys [{interpMode=RCIM_Cubic, tangentMode=RCTM_User, time=0, value=1}, {interpMode=RCIM_Cubic, tangentMode=RCTM_User, time=1, value=0}]}}}
     Scale Curve=&FNTimeOfDayScalar(Night Scale=1, Morning Scale=1, Evening Scale=1)
    )
    Scale RGB=&MakeVectorFromLinearColorRGB(Color=&LerpLinearColors(
      StartColor=<rgba 5 0.6247 0 1>
      EndColor=<rgba 0.2865 0 0 1>
      LerpFactor=&FloatFromCurve(FloatCurve=DataInterface() {propertyValues {Curve {keys [
           {arriveTangent=19.3856, interpMode=RCIM_Cubic, leaveTangent=19.3856, tangentMode=RCTM_User, time=0, value=0}
           {arriveTangent=1.3591, interpMode=RCIM_Cubic, leaveTangent=1.3591, tangentMode=RCTM_User, time=0.1, value=0.9}
           {interpMode=RCIM_Cubic, tangentMode=RCTM_User, time=0.4, value=1}
          ]}}})
     ))
    ScaleA=true
    ScaleRGB=true
   )
  ]
  rendererClasses [&NiagaraRibbonRendererProperties]
  renderers [{rendererClass=&NiagaraRibbonRendererProperties}]
 }
 DEBUG {
  bEnabled=false
  emitterUpdateScript [
   &EmitterState()
   &SpawnRate(SpawnRate=30, Use Spawn Probability=false)
  ]
  particleSpawnScript [&InitializeParticle(Color=<rgba 0.9834 0 1 1>, Sprite Size Mode=::NewEnumerator3, UsePositionOffset=false, Write Material Random=true)]
  particleUpdateScript [
   &ParticleState()
   &SkeletalMeshLocation(
    Bone Sampling Mode=::NewEnumerator0
    "Inherit Velocity (Scale)"=-0.35
    Inherited Velocity Speed Limit=50
    MaskSpawnGroup=false
    OverrideRandomSeed=false
    Sampled Position Offset=&RandomRangeVector(Minimum=<0 0 -10>, Maximum=<0 0 10>)
    Skeletal Mesh=$User.Skeletal Mesh
    Velocity Sampling=::NewEnumerator0
    VelocitySpeedLimit=true
   )
   &InheritVelocity(Inherited Velocity Amount Scale=<0.3 0.3 0.3>, Inherited Velocity Speed Limit=1000, Speed Limit=true, UseSpeedThreshold=false) {enabled=false}
   &SolveForcesAndVelocity(Clamp Velocity=false, Limit Acceleration=false)
  ]
  rendererClasses [&NiagaraSpriteRendererProperties]
  renderers [{rendererClass=&NiagaraSpriteRendererProperties}]
 }
 Particle_Smoke {
  emitterUpdateScript [
   &EmitterState()
   &SpawnRate(SpawnRate=10, Use Spawn Probability=false)
  ]
  particleSpawnScript [
   /SetVariables(Particles.RandCurve:NiagaraFloat=&FloatFromCurve(
     FloatCurve=DataInterface() {propertyValues {Curve {keys [
         {interpMode=RCIM_Cubic, tangentMode=RCTM_User, time=0, value=0}
         {arriveTangent=0.7625, interpMode=RCIM_Cubic, leaveTangent=0.7625, tangentMode=RCTM_User, time=0.8, value=0.2}
         {arriveTangent=7.0993, interpMode=RCIM_Cubic, leaveTangent=7.0993, tangentMode=RCTM_User, time=1, value=1}
        ]}}}
     CurveIndex=&RandomRangeFloat(Minimum=0, Maximum=1)
    )) {isSetParameters=true}
   &InitializeParticle(
    Color=&LerpLinearColors(
     StartColor=<rgba 0.0417 0.0049 0 0.1>
     EndColor=<rgba 0 0 0 0.1>
     LerpFactor=&FloatFromCurve(
      FloatCurve=DataInterface() {propertyValues {Curve {keys [
          {interpMode=RCIM_Cubic, tangentMode=RCTM_User, time=0, value=0}
          {arriveTangent=0.7625, interpMode=RCIM_Cubic, leaveTangent=0.7625, tangentMode=RCTM_User, time=0.9, value=0.2}
          {arriveTangent=12.0097, interpMode=RCIM_Cubic, leaveTangent=12.0097, tangentMode=RCTM_User, time=1, value=1}
         ]}}}
      CurveIndex=&RandomRangeFloat(Minimum=0, Maximum=1)
     )
    )
    Lifetime Max=1.2
    Lifetime Min=0.9
    Lifetime Mode=::NewEnumerator1
    Sprite Rotation Mode=::NewEnumerator2
    Sprite Size=&Multiply_Vector2D_ByFloat(Float=&Lerp_Float(A=1, B=2, Alpha=$Particles.RandCurve), Vector2D=<36 32>)
    Sprite Size Mode=::NewEnumerator4
    UsePositionOffset=false
    Write Material Random=true
   )
   &InitializeParticle(
    Color=&LerpLinearColors(
     StartColor=<rgba 0.0417 0.0006 0 0.1>
     EndColor=<rgba 0 0 0 0.1>
     LerpFactor=&FloatFromCurve(
      FloatCurve=DataInterface() {propertyValues {Curve {keys [
          {interpMode=RCIM_Cubic, tangentMode=RCTM_User, time=0, value=0}
          {arriveTangent=0.7625, interpMode=RCIM_Cubic, leaveTangent=0.7625, tangentMode=RCTM_User, time=0.9, value=0.2}
          {arriveTangent=12.0097, interpMode=RCIM_Cubic, leaveTangent=12.0097, tangentMode=RCTM_User, time=1, value=1}
         ]}}}
      CurveIndex=&RandomRangeFloat(Minimum=0, Maximum=1)
     )
    )
    Lifetime Max=1.2
    Lifetime Min=0.9
    Lifetime Mode=::NewEnumerator1
    Sprite Rotation Mode=::NewEnumerator2
    Sprite Size=&Multiply_Vector2D_ByFloat(Float=&Lerp_Float(A=1, B=2, Alpha=$Particles.RandCurve), Vector2D=<36 32>)
    Sprite Size Mode=::NewEnumerator4
    UsePositionOffset=false
    Write Material Random=true
   ) {enabled=false, module=InitializeParticle001}
   &CameraOffset(Camera Offset Amount=&RandomRangeFloat(Minimum=5, Maximum=15))
  ]
  particleUpdateScript [
   &ParticleState()
   &SkeletalMeshLocation(
    Bone Sampling Mode=::NewEnumerator0
    "Inherit Velocity (Scale)":NiagaraFloat=&Lerp_Float(A=-0.35, B=-0.6, Alpha=$Particles.RandCurve)
    Inherited Velocity Speed Limit:NiagaraFloat=&Lerp_Float(A=50, B=120, Alpha=$Particles.RandCurve)
    MaskSpawnGroup=false
    OverrideRandomSeed=false
    Sampled Position Offset=&RandomRangeVector(Minimum=<0 0 -10>, Maximum=<0 0 10>)
    Skeletal Mesh=$User.Skeletal Mesh
    Velocity Sampling=::NewEnumerator0
    VelocitySpeedLimit=true
   )
   &SubUVAnimation(SubUV Lookup Index:NiagaraFloat=&Multiply_Float(A=$Particles.NormalizedAge, B=0.25))
   &ScaleColor(
    Scale Alpha=&FloatFromCurve(
     FloatCurve=DataInterface() {propertyValues {Curve {keys [
         {interpMode=RCIM_Cubic, tangentMode=RCTM_User, time=0, value=0}
         {interpMode=RCIM_Cubic, tangentMode=RCTM_User, time=0.4, value=1}
         {interpMode=RCIM_Cubic, tangentMode=RCTM_User, time=0.7, value=1}
         {interpMode=RCIM_Cubic, tangentMode=RCTM_User, time=1, value=0}
        ]}}}
     Scale Curve=&Lerp_Float(
      A=1
      B=1.5
      Alpha=&FloatFromCurve(
       FloatCurve=DataInterface() {propertyValues {Curve {keys [
           {interpMode=RCIM_Cubic, tangentMode=RCTM_User, time=0, value=0}
           {arriveTangent=0.7625, interpMode=RCIM_Cubic, leaveTangent=0.7625, tangentMode=RCTM_User, time=0.9, value=0.2}
           {arriveTangent=12.0097, interpMode=RCIM_Cubic, leaveTangent=12.0097, tangentMode=RCTM_User, time=1, value=1}
          ]}}}
       CurveIndex=&RandomRangeFloat(Minimum=0, Maximum=1)
      )
     )
    )
    ScaleA=true
    ScaleRGB=false
   )
   &SolveForcesAndVelocity(Clamp Velocity=false, Limit Acceleration=false)
  ]
  rendererClasses [&NiagaraSpriteRendererProperties]
  renderers [{rendererClass=&NiagaraSpriteRendererProperties}]
 }
}
summary {
 systemName=NS_Katana_Vampire_Ambient
 emitters [
  {emitterName=Particle_Smoke, rendererClasses [&NiagaraSpriteRendererProperties]}
  {bEnabled=false, emitterName=DEBUG, rendererClasses [&NiagaraSpriteRendererProperties]}
  {emitterName=AnimTrail, rendererClasses [&NiagaraRibbonRendererProperties]}
  {emitterName=AnimTrailLower, rendererClasses [&NiagaraRibbonRendererProperties]}
 ]
 userVariables [
  User.FX Stage {type=NiagaraFloat, defaultValue=1, description=""}
  User.RibbonIDOffset {type=NiagaraInt32}
  User.Skeletal Mesh {
   type=NiagaraDataInterfaceSkeletalMesh
   defaultValue {
    struct=NiagaraExt_VariableValue_DataInterface
    value {dataInterface=&self:NiagaraDataInterfaceSkeletalMesh_0}
   }
  }
 ]
}
