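# Sound cue for the Vampire Katana, named as the signal that its charge attack is available.
# A top mixer layers two charge-available waves, a random dash-charge sound and a boosted, pitched-down random idle voice.
# A quality-level switch gives the first two levels six voice variants and the last two three.
# It uses weapon-pickup attenuation, with concurrency allowing one instance and blocking new ones.
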SoundCue /VampireKatanaGameplay/Items/VampireKatana/Sound/Vampire_Katana_Attack_Charge_Available_Cue
root=Mixer_1
nodes [
 {id=Mixer_0, children [WavePlayer_0,WavePlayer_1], properties {inputVolume [1,0.5]}}
 {id=Mixer_1, children [Modulator_0,Modulator_1,Modulator_5], properties {inputVolume [1,1,2.5]}}
 {id=Modulator_0, children [Mixer_0], properties {pitchMax=0.9, pitchMin=0.9}}
 {id=Modulator_1, children [Random_0], properties {pitchMax=1, pitchMin=1}}
 {id=Modulator_5, children [QualityLevel_0], properties {pitchMax=0.8, pitchMin=0.8, volumeMax=1, volumeMin=1}}
 {id=QualityLevel_0, children [Random_7,Random_7,Random_1,Random_1]}
 {
  id=Random_0
  children [WavePlayer_2,WavePlayer_3,WavePlayer_4,WavePlayer_5]
  properties {bShouldExcludeFromBranchCulling=true, weights [1,1,1,1]}
 }
 {id=Random_1, children [WavePlayer_9, WavePlayer_10, WavePlayer_11], properties {bShouldExcludeFromBranchCulling=true, weights [1,1,1]}}
 {
  id=Random_7
  children [WavePlayer_6, WavePlayer_7, WavePlayer_8, WavePlayer_9, WavePlayer_10, WavePlayer_11]
  properties {weights [1,1,1,1,1,1]}
 }
 {id=WavePlayer_0, properties {soundWaveAssetPtr=&/Vampire_Katana_Attack_Charge_Available_04}}
 {id=WavePlayer_1, properties {soundWaveAssetPtr=&/Vampire_Katana_Attack_Charge_Available_05}}
 {id=WavePlayer_10, properties {soundWaveAssetPtr=&/Vampire_Katana_Idle_Voices_02}}
 {id=WavePlayer_11, properties {soundWaveAssetPtr=&/Vampire_Katana_Idle_Voices_01}}
 {id=WavePlayer_2, properties {soundWaveAssetPtr=&/Vampire_Katana_DashCharge_01}}
 {id=WavePlayer_3, properties {soundWaveAssetPtr=&/Vampire_Katana_DashCharge_02}}
 {id=WavePlayer_4, properties {soundWaveAssetPtr=&/Vampire_Katana_DashCharge_03}}
 {id=WavePlayer_5, properties {soundWaveAssetPtr=&/Vampire_Katana_DashCharge_04}}
 {id=WavePlayer_6, properties {soundWaveAssetPtr=&/Vampire_Katana_Idle_Voices_06}}
 {id=WavePlayer_7, properties {soundWaveAssetPtr=&/Vampire_Katana_Idle_Voices_05}}
 {id=WavePlayer_8, properties {soundWaveAssetPtr=&/Vampire_Katana_Idle_Voices_04}}
 {id=WavePlayer_9, properties {soundWaveAssetPtr=&/Vampire_Katana_Idle_Voices_03}}
]
properties {
 attenuationSettings=&Fort_Weapon_Pickup_Att
 bOverrideConcurrency=true
 maxDistance=3500
 soundClassObject=&Fort_Loot_Collect
 volumeMultiplier=0.498
 concurrencyOverrides {maxCount=1, resolutionRule=PreventNew}
}
