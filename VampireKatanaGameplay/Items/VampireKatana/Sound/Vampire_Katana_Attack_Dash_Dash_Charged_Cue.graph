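# Sound cue for the Vampire Katana's charged dash attack, with its cue-level volume multiplier set to zero.
# It mixes a modulated random pick of three charged-dash waves with a random pick of three dash-swing waves delayed a quarter second.
# The mix passes through a Doppler node at the root.
# It uses the base Katana's dash attenuation asset, a melee weapon sound class and a global-limit dash concurrency setting.
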SoundCue /VampireKatanaGameplay/Items/VampireKatana/Sound/Vampire_Katana_Attack_Dash_Dash_Charged_Cue
root=Doppler_0
nodes [
 {id=Delay_0, children [Random_1], properties {delayMax=0.25, delayMin=0.25}}
 {id=Doppler_0, children [Mixer_0]}
 {id=Mixer_0, children [Modulator_0,Delay_0], properties {inputVolume [1,1]}}
 {id=Modulator_0, children [Random_0]}
 {id=Random_0, children [WavePlayer_5,WavePlayer_7,WavePlayer_8], properties {weights [1,1,1]}}
 {id=Random_1, children [WavePlayer_12, WavePlayer_13, WavePlayer_14], properties {weights [1,1,1]}}
 {id=WavePlayer_12, properties {soundWaveAssetPtr=&/Vampire_Katana_Dash_Swing_01}}
 {id=WavePlayer_13, properties {soundWaveAssetPtr=&/Vampire_Katana_Dash_Swing_02}}
 {id=WavePlayer_14, properties {soundWaveAssetPtr=&/Vampire_Katana_Dash_Swing_03}}
 {id=WavePlayer_5, properties {soundWaveAssetPtr=&/Vampire_Katana_Attack_Dash_Dash_Charged_01}}
 {id=WavePlayer_7, properties {soundWaveAssetPtr=&/Vampire_Katana_Attack_Dash_Dash_Charged_02}}
 {id=WavePlayer_8, properties {soundWaveAssetPtr=&/Vampire_Katana_Attack_Dash_Dash_Charged_03}}
]
properties {
 attenuationSettings=/KatanaGameplay/Items/Katana/Sounds/Fort_Weapon_Melee_Katana_Dash_Linear_Att
 maxDistance=16000
 soundClassObject=&Fort_WeaponMelee_PSM
 volumeMultiplier=0
 attenuationOverrides {spatializationAlgorithm=SPATIALIZATION_HRTF}
 concurrencySet [/VampireKatanaGameplay/Items/VampireKatana/Sound/Settings/VampireKatana_Dash_GlobalLimit_Conc]
}
