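# Sound cue for the Vampire Katana's dash attack, played through the melee weapon sound class with a dash concurrency limit and base Katana attenuation.
# A quality switch selects between two Doppler-wrapped mixes, the second using fewer wave variations and different delays.
# Each mixes randomized, modulated dash layers with delayed swing and sweetener layers.
# The DashB layer is wired in but muted at zero mixer volume in both mixes.
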SoundCue /VampireKatanaGameplay/Items/VampireKatana/Sound/Vampire_Katana_Attack_Dash_Dash_Cue
root=QualityLevel_0
nodes [
 {id=Delay_0, children [Random_1], properties {delayMax=0.2, delayMin=0.2}}
 {id=Delay_1, children [Random_3], properties {delayMax=0.25, delayMin=0.25}}
 {id=Delay_2, children [Modulator_6], properties {delayMax=0.4, delayMin=0.4}}
 {id=Delay_4, children [Modulator_9], properties {delayMax=0.3, delayMin=0.3}}
 {id=Doppler_0, children [Mixer_0]}
 {id=Doppler_1, children [Mixer_1]}
 {id=Mixer_0, children [Modulator_0,Modulator_2,Modulator_1,Delay_4], properties {inputVolume [1.4,0,1.5,2]}}
 {id=Mixer_1, children [Modulator_3,Modulator_5,Modulator_4,Delay_2], properties {inputVolume [1.4,0,1.5,1]}}
 {id=Modulator_0, children [Random_5], properties {volumeMax=1, volumeMin=0.9}}
 {id=Modulator_1, children [Delay_0]}
 {id=Modulator_2, children [Random_0], properties {pitchMax=1.2, pitchMin=1.1}}
 {id=Modulator_3, children [Random_4], properties {volumeMax=1, volumeMin=0.9}}
 {id=Modulator_4, children [Delay_1]}
 {id=Modulator_5, children [Random_2], properties {pitchMax=1.2, pitchMin=1.1}}
 {id=Modulator_6, children [Random_7], properties {pitchMax=1.1}}
 {id=Modulator_9, children [Random_6], properties {pitchMax=1.1}}
 {id=QualityLevel_0, children [Doppler_0,Doppler_0,Doppler_1,Doppler_1]}
 {id=Random_0, children [WavePlayer_16, WavePlayer_20, WavePlayer_21, WavePlayer_22], properties {weights [1,1,1,1]}}
 {id=Random_1, children [WavePlayer_23, WavePlayer_24, WavePlayer_25], properties {weights [1,1,1]}}
 {id=Random_2, children [WavePlayer_29, WavePlayer_30], properties {bShouldExcludeFromBranchCulling=true, weights [1,1]}}
 {id=Random_3, children [WavePlayer_31, WavePlayer_32], properties {bShouldExcludeFromBranchCulling=true, weights [1,1]}}
 {id=Random_4, children [WavePlayer_26, WavePlayer_27], properties {bShouldExcludeFromBranchCulling=true, weights [1,1]}}
 {id=Random_5, children [WavePlayer_9, WavePlayer_10, WavePlayer_13], properties {weights [1,1,1]}}
 {id=Random_6, children [WavePlayer_2,WavePlayer_3,WavePlayer_5,WavePlayer_6], properties {weights [1,1,1,1]}}
 {id=Random_7, children [WavePlayer_15, WavePlayer_17], properties {bShouldExcludeFromBranchCulling=true, weights [1,1]}}
 {id=WavePlayer_10, properties {soundWaveAssetPtr=&/Vampire_Katana_Attack_Dash_DashC_02}}
 {id=WavePlayer_13, properties {soundWaveAssetPtr=&/Vampire_Katana_Attack_Dash_DashC_03}}
 {id=WavePlayer_15, properties {soundWaveAssetPtr=&/Vampire_Katana_Dash_Sweetner_03}}
 {id=WavePlayer_16, properties {soundWaveAssetPtr=&/Vampire_Katana_Attack_Dash_DashB_04}}
 {id=WavePlayer_17, properties {soundWaveAssetPtr=&/Vampire_Katana_Dash_Sweetner_04}}
 {id=WavePlayer_2, properties {soundWaveAssetPtr=&/Vampire_Katana_Dash_Sweetner_01}}
 {id=WavePlayer_20, properties {soundWaveAssetPtr=&/Vampire_Katana_Attack_Dash_DashB_01}}
 {id=WavePlayer_21, properties {soundWaveAssetPtr=&/Vampire_Katana_Attack_Dash_DashB_02}}
 {id=WavePlayer_22, properties {soundWaveAssetPtr=&/Vampire_Katana_Attack_Dash_DashB_03}}
 {id=WavePlayer_23, properties {soundWaveAssetPtr=&/Vampire_Katana_Dash_Swing_01}}
 {id=WavePlayer_24, properties {soundWaveAssetPtr=&/Vampire_Katana_Dash_Swing_02}}
 {id=WavePlayer_25, properties {soundWaveAssetPtr=&/Vampire_Katana_Dash_Swing_03}}
 {id=WavePlayer_26, properties {soundWaveAssetPtr=&/Vampire_Katana_Attack_Dash_DashC_01}}
 {id=WavePlayer_27, properties {soundWaveAssetPtr=&Vampire_Katana_Attack_Dash_DashC_02}}
 {id=WavePlayer_29, properties {soundWaveAssetPtr=&Vampire_Katana_Attack_Dash_DashB_01}}
 {id=WavePlayer_3, properties {soundWaveAssetPtr=&/Vampire_Katana_Dash_Sweetner_02}}
 {id=WavePlayer_30, properties {soundWaveAssetPtr=&Vampire_Katana_Attack_Dash_DashB_02}}
 {id=WavePlayer_31, properties {soundWaveAssetPtr=&Vampire_Katana_Dash_Swing_01}}
 {id=WavePlayer_32, properties {soundWaveAssetPtr=&Vampire_Katana_Dash_Swing_02}}
 {id=WavePlayer_5, properties {soundWaveAssetPtr=&Vampire_Katana_Dash_Sweetner_03}}
 {id=WavePlayer_6, properties {soundWaveAssetPtr=&Vampire_Katana_Dash_Sweetner_04}}
 {id=WavePlayer_9, properties {soundWaveAssetPtr=&Vampire_Katana_Attack_Dash_DashC_01}}
]
properties {
 attenuationSettings=/KatanaGameplay/Items/Katana/Sounds/Fort_Weapon_Melee_Katana_Dash_Linear_Att
 maxDistance=16000
 soundClassObject=&Fort_WeaponMelee_PSM
 volumeMultiplier=1
 attenuationOverrides {spatializationAlgorithm=SPATIALIZATION_HRTF}
 concurrencySet [/VampireKatanaGameplay/Items/VampireKatana/Sound/Settings/VampireKatana_Dash_GlobalLimit_Conc]
}
