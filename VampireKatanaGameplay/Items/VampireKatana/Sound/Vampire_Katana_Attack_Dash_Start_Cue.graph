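# Sound cue for the start of the Vampire Katana's dash attack.
# A Random node picks one of three equally weighted StartB wave variations.
# It uses the base Katana's linear dash attenuation and the melee weapon sound class, with a slight volume boost.
# A Vampire Katana dash concurrency asset, named as a global limit, governs simultaneous playback.
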SoundCue /VampireKatanaGameplay/Items/VampireKatana/Sound/Vampire_Katana_Attack_Dash_Start_Cue
root=Random_1
nodes [
 {id=Random_1, children [WavePlayer_3,WavePlayer_4,WavePlayer_5], properties {weights [1,1,1]}}
 {id=WavePlayer_3, properties {soundWaveAssetPtr=&/Vampire_Katana_Attack_Dash_StartB_01}}
 {id=WavePlayer_4, properties {soundWaveAssetPtr=&/Vampire_Katana_Attack_Dash_StartB_02}}
 {id=WavePlayer_5, properties {soundWaveAssetPtr=&/Vampire_Katana_Attack_Dash_StartB_03}}
]
properties {
 attenuationSettings=/KatanaGameplay/Items/Katana/Sounds/Fort_Weapon_Melee_Katana_Dash_Linear_Att
 maxDistance=16000
 soundClassObject=&Fort_WeaponMelee_PSM
 volumeMultiplier=1.3
 attenuationOverrides {spatializationAlgorithm=SPATIALIZATION_HRTF}
 concurrencySet [/VampireKatanaGameplay/Items/VampireKatana/Sound/Settings/VampireKatana_Dash_GlobalLimit_Conc]
}
