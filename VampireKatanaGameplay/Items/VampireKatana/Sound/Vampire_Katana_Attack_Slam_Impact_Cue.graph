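# Sound cue for the Vampire Katana slam attack's impact, built as a layered mix of randomized variations.
# A mixer blends four layers at different volumes: ground impact, boom, a lower-pitched second-swing sound, and a swing sweetener delayed slightly.
# Each layer picks a random wave and passes through a modulator for pitch, with the sweetener also varying volume.
# Playback uses close-range melee attenuation, the melee weapon sound class, and the katana dash global-limit concurrency.
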SoundCue /VampireKatanaGameplay/Items/VampireKatana/Sound/Vampire_Katana_Attack_Slam_Impact_Cue
root=Mixer_4
nodes [
 {id=Delay_5, children [Modulator_10], properties {delayMax=0.1, delayMin=0.1}}
 {id=Mixer_4, children [Modulator_1,Modulator_0,Modulator_2,Delay_5], properties {inputVolume [1.3,0.5,1,0.7]}}
 {id=Modulator_0, children [Random_8], properties {pitchMax=1.3, pitchMin=1}}
 {id=Modulator_1, children [Random_7], properties {pitchMax=1.3, pitchMin=1}}
 {id=Modulator_10, children [Random_11], properties {pitchMax=1.1, volumeMax=1, volumeMin=0.9}}
 {id=Modulator_2, children [Random_1], properties {pitchMax=0.9, pitchMin=0.9}}
 {id=Random_1, children [WavePlayer_8, WavePlayer_9, WavePlayer_10, WavePlayer_11], properties {weights [1,1,1,1]}}
 {id=Random_11, children [WavePlayer_15, WavePlayer_17, WavePlayer_18], properties {bShouldExcludeFromBranchCulling=true, weights [1,1,1]}}
 {id=Random_7, children [WavePlayer_0,WavePlayer_1,WavePlayer_2], properties {weights [1,1,1]}}
 {id=Random_8, children [WavePlayer_3,WavePlayer_4,WavePlayer_5], properties {weights [1,1,1]}}
 {id=WavePlayer_0, properties {soundWaveAssetPtr=&/Vampire_Katana_Attack_Slam_Impact_Ground_01}}
 {id=WavePlayer_1, properties {soundWaveAssetPtr=&/Vampire_Katana_Attack_Slam_Impact_Ground_02}}
 {id=WavePlayer_10, properties {soundWaveAssetPtr=&/Vampire_Katana_Attack_Swing_Second_A_03}}
 {id=WavePlayer_11, properties {soundWaveAssetPtr=&/Vampire_Katana_Attack_Swing_Second_A_04}}
 {id=WavePlayer_15, properties {soundWaveAssetPtr=&/Vampire_Katana_SwingSweetner_01}}
 {id=WavePlayer_17, properties {soundWaveAssetPtr=&/Vampire_Katana_SwingSweetner_02}}
 {id=WavePlayer_18, properties {soundWaveAssetPtr=&/Vampire_Katana_SwingSweetner_03}}
 {id=WavePlayer_2, properties {soundWaveAssetPtr=&/Vampire_Katana_Attack_Slam_Impact_Ground_03}}
 {id=WavePlayer_3, properties {soundWaveAssetPtr=&/Vampire_Katana_Attack_Slam_Impact_Boom_01}}
 {id=WavePlayer_4, properties {soundWaveAssetPtr=&/Vampire_Katana_Attack_Slam_Impact_Boom_02}}
 {id=WavePlayer_5, properties {soundWaveAssetPtr=&/Vampire_Katana_Attack_Slam_Impact_Boom_03}}
 {id=WavePlayer_8, properties {soundWaveAssetPtr=&/Vampire_Katana_Attack_Swing_Second_A_01}}
 {id=WavePlayer_9, properties {soundWaveAssetPtr=&/Vampire_Katana_Attack_Swing_Second_A_02}}
]
properties {
 attenuationSettings=/Game/Sounds/Attenuation/Fort_Weapon_Melee_Attack_Close_Range_Stereo_Att
 maxDistance=10058.75
 soundClassObject=&Fort_WeaponMelee_PSM
 volumeMultiplier=1
 attenuationOverrides {spatializationAlgorithm=SPATIALIZATION_HRTF}
 concurrencySet [/VampireKatanaGameplay/Items/VampireKatana/Sound/Settings/VampireKatana_Dash_GlobalLimit_Conc]
}
