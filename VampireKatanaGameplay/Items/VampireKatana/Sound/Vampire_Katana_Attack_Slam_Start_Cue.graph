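# Sound cue for the start of the Vampire Katana's slam attack, using the katana's close-range attack attenuation with HRTF spatialization and a dash concurrency limit.
# It mixes a random dash swing, with Doppler and pitch and volume variation, with a briefly delayed, pitch-varied swing sweetener.
# A quality-level switch picks the sweetener from all six variations or a three-variation subset, reducing variety on some quality levels.
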SoundCue /VampireKatanaGameplay/Items/VampireKatana/Sound/Vampire_Katana_Attack_Slam_Start_Cue
root=Mixer_0
nodes [
 {id=Delay_4, children [Modulator_9], properties {delayMax=0.15, delayMin=0.15}}
 {id=Doppler_0, children [Modulator_1]}
 {id=Mixer_0, children [Doppler_0,Delay_4], properties {inputVolume [1,1]}}
 {id=Modulator_1, children [Random_1], properties {pitchMax=1.1, volumeMax=1.3, volumeMin=1.1}}
 {id=Modulator_9, children [QualityLevel_0], properties {pitchMax=1.1}}
 {id=QualityLevel_0, children [Random_5,Random_5,Random_0,Random_0]}
 {id=Random_0, children [WavePlayer_12, WavePlayer_4, WavePlayer_14], properties {bShouldExcludeFromBranchCulling=true, weights [1,1,1]}}
 {id=Random_1, children [WavePlayer_0,WavePlayer_1,WavePlayer_5], properties {weights [1,1,1]}}
 {
  id=Random_5
  children [WavePlayer_9, WavePlayer_2, WavePlayer_11, WavePlayer_12, WavePlayer_4, WavePlayer_14]
  properties {weights [1,1,1,1,1,1]}
 }
 {id=WavePlayer_0, properties {soundWaveAssetPtr=&/Vampire_Katana_Dash_Swing_01}}
 {id=WavePlayer_1, properties {soundWaveAssetPtr=&/Vampire_Katana_Dash_Swing_02}}
 {id=WavePlayer_11, properties {soundWaveAssetPtr=&/Vampire_Katana_Swing2Sweetner_03}}
 {id=WavePlayer_12, properties {soundWaveAssetPtr=&/Vampire_Katana_Swing2Sweetner_04}}
 {id=WavePlayer_14, properties {soundWaveAssetPtr=&/Vampire_Katana_Swing2Sweetner_06}}
 {id=WavePlayer_2, properties {soundWaveAssetPtr=&/Vampire_Katana_Swing2Sweetner_02}}
 {id=WavePlayer_4, properties {soundWaveAssetPtr=&/Vampire_Katana_Swing2Sweetner_05}}
 {id=WavePlayer_5, properties {soundWaveAssetPtr=&/Vampire_Katana_Dash_Swing_03}}
 {id=WavePlayer_9, properties {soundWaveAssetPtr=&/Vampire_Katana_Swing2Sweetner_01}}
]
properties {
 attenuationSettings=/KatanaGameplay/Items/Katana/Sounds/Fort_Weapon_Melee_Attack_Close_Katana_Att
 maxDistance=7200
 soundClassObject=&Fort_WeaponMelee_PSM
 volumeMultiplier=0.6
 attenuationOverrides {spatializationAlgorithm=SPATIALIZATION_HRTF}
 concurrencySet [/VampireKatanaGameplay/Items/VampireKatana/Sound/Settings/VampireKatana_Dash_GlobalLimit_Conc]
}
