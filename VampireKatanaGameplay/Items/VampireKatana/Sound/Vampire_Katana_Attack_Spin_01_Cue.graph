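# Sound cue for the Vampire Katana's spin attack, with slight random pitch variation and owner-limited concurrency.
# A quality-level node sends its first two slots to a mix with larger random wave pools and its last two to a mix with two-wave pools.
# Each mix layers first-swing, secondary swing, power swing and sweetener sounds, with short delays staggering the first-swing and sweetener layers.
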SoundCue /VampireKatanaGameplay/Items/VampireKatana/Sound/Vampire_Katana_Attack_Spin_01_Cue
root=Modulator_0
nodes [
 {id=Delay_1, children [Random_6], properties {delayMax=0.1, delayMin=0.1}}
 {id=Delay_2, children [Random_2], properties {delayMax=0.1, delayMin=0.1}}
 {id=Delay_3, children [Modulator_10], properties {delayMax=0.15, delayMin=0.15}}
 {id=Delay_4, children [Modulator_9], properties {delayMax=0.15, delayMin=0.15}}
 {id=Mixer_0, children [Modulator_5,Modulator_7,Modulator_12,Delay_3], properties {inputVolume [1,0.5,0.9,0.5]}}
 {id=Mixer_2, children [Modulator_1,Modulator_8,Modulator_3,Delay_4], properties {inputVolume [1,0.4,1,1]}}
 {id=Modulator_0, children [QualityLevel_0], properties {pitchMax=1.1, pitchMin=1, volumeMin=1}}
 {id=Modulator_1, children [Delay_1]}
 {id=Modulator_10, children [Random_8], properties {pitchMax=1.1}}
 {id=Modulator_12, children [Random_4]}
 {id=Modulator_3, children [Random_0]}
 {id=Modulator_4, children [Mixer_0]}
 {id=Modulator_5, children [Delay_2]}
 {id=Modulator_6, children [Mixer_2]}
 {id=Modulator_7, children [Random_3], properties {pitchMax=1, pitchMin=0.9}}
 {id=Modulator_8, children [Random_7], properties {pitchMax=1, pitchMin=0.9}}
 {id=Modulator_9, children [Random_5], properties {pitchMax=1.1}}
 {id=QualityLevel_0, children [Modulator_6,Modulator_6,Modulator_4,Modulator_4]}
 {id=Random_0, children [WavePlayer_25, WavePlayer_26, WavePlayer_27, WavePlayer_28, WavePlayer_29], properties {weights [1,1,1,0.5,1]}}
 {id=Random_2, children [WavePlayer_30, WavePlayer_31], properties {bShouldExcludeFromBranchCulling=true, weights [1,1]}}
 {id=Random_3, children [WavePlayer_32, WavePlayer_33], properties {bShouldExcludeFromBranchCulling=true, weights [1,1]}}
 {id=Random_4, children [WavePlayer_34, WavePlayer_35], properties {bShouldExcludeFromBranchCulling=true, weights [1,1]}}
 {
  id=Random_5
  children [
   WavePlayer_9
   WavePlayer_0
   WavePlayer_11
   WavePlayer_12
   WavePlayer_4
   WavePlayer_14
   WavePlayer_1
   WavePlayer_2
   WavePlayer_3
   WavePlayer_16
  ]
  properties {weights [1,1,1,1,1,1,1,1,1,1]}
 }
 {id=Random_6, children [WavePlayer_5, WavePlayer_8, WavePlayer_10], properties {weights [1,1,1]}}
 {id=Random_7, children [WavePlayer_13, WavePlayer_15, WavePlayer_19, WavePlayer_20], properties {weights [1,1,1,1]}}
 {id=Random_8, children [WavePlayer_6,WavePlayer_7], properties {bShouldExcludeFromBranchCulling=true, weights [1,1]}}
 {id=WavePlayer_0, properties {soundWaveAssetPtr=&/Vampire_Katana_Swing2Sweetner_02}}
 {id=WavePlayer_1, properties {soundWaveAssetPtr=&/Vampire_Katana_Swing2Sweetner_10}}
 {id=WavePlayer_10, properties {soundWaveAssetPtr=&/Vampire_Katana_Attack_Swing_First_03}}
 {id=WavePlayer_11, properties {soundWaveAssetPtr=&/Vampire_Katana_Swing2Sweetner_03}}
 {id=WavePlayer_12, properties {soundWaveAssetPtr=&/Vampire_Katana_Swing2Sweetner_04}}
 {
  id=WavePlayer_13
  properties {soundWaveAssetPtr=&/Vampire_Katana_Attack_Swing_First_Layer_B_01}
 }
 {id=WavePlayer_14, properties {soundWaveAssetPtr=&/Vampire_Katana_Swing2Sweetner_06}}
 {
  id=WavePlayer_15
  properties {soundWaveAssetPtr=&/Vampire_Katana_Attack_Swing_First_Layer_B_02}
 }
 {id=WavePlayer_16, properties {soundWaveAssetPtr=&/Vampire_Katana_Swing2Sweetner_07}}
 {
  id=WavePlayer_19
  properties {soundWaveAssetPtr=&/Vampire_Katana_Attack_Swing_First_Layer_B_03}
 }
 {id=WavePlayer_2, properties {soundWaveAssetPtr=&/Vampire_Katana_Swing2Sweetner_09}}
 {
  id=WavePlayer_20
  properties {soundWaveAssetPtr=&/Vampire_Katana_Attack_Swing_First_Layer_B_04}
 }
 {id=WavePlayer_25, properties {soundWaveAssetPtr=&/Vampire_Katana_Swing_Power_A_01}}
 {id=WavePlayer_26, properties {soundWaveAssetPtr=&/Vampire_Katana_Swing_Power_A_02}}
 {id=WavePlayer_27, properties {soundWaveAssetPtr=&/Vampire_Katana_Swing_Power_A_03}}
 {id=WavePlayer_28, properties {soundWaveAssetPtr=&/Vampire_Katana_Swing_Power_C_03}}
 {id=WavePlayer_29, properties {soundWaveAssetPtr=&/Vampire_Katana_Swing_Power_C_02}}
 {id=WavePlayer_3, properties {soundWaveAssetPtr=&/Vampire_Katana_Swing2Sweetner_08}}
 {id=WavePlayer_30, properties {soundWaveAssetPtr=&/Vampire_Katana_Attack_Swing_First_01}}
 {id=WavePlayer_31, properties {soundWaveAssetPtr=&/Vampire_Katana_Attack_Swing_First_02}}
 {
  id=WavePlayer_32
  properties {soundWaveAssetPtr=&Vampire_Katana_Attack_Swing_First_Layer_B_01}
 }
 {
  id=WavePlayer_33
  properties {soundWaveAssetPtr=&Vampire_Katana_Attack_Swing_First_Layer_B_02}
 }
 {id=WavePlayer_34, properties {soundWaveAssetPtr=&Vampire_Katana_Swing_Power_A_01}}
 {id=WavePlayer_35, properties {soundWaveAssetPtr=&Vampire_Katana_Swing_Power_A_02}}
 {id=WavePlayer_4, properties {soundWaveAssetPtr=&/Vampire_Katana_Swing2Sweetner_05}}
 {id=WavePlayer_5, properties {soundWaveAssetPtr=&Vampire_Katana_Attack_Swing_First_01}}
 {id=WavePlayer_6, properties {soundWaveAssetPtr=&/Vampire_Katana_Swing2Sweetner_01}}
 {id=WavePlayer_7, properties {soundWaveAssetPtr=&Vampire_Katana_Swing2Sweetner_02}}
 {id=WavePlayer_8, properties {soundWaveAssetPtr=&Vampire_Katana_Attack_Swing_First_02}}
 {id=WavePlayer_9, properties {soundWaveAssetPtr=&Vampire_Katana_Swing2Sweetner_01}}
]
properties {
 attenuationSettings=/KatanaGameplay/Items/Katana/Sounds/Fort_Weapon_Melee_Attack_Close_Katana_Att
 bOverrideConcurrency=true
 maxDistance=7200
 soundClassObject=&Fort_WeaponMelee_PSM
 volumeMultiplier=0.8
 attenuationOverrides {
  binauralRadius=500
  falloffDistance=2400
  nonSpatializedRadiusStart=500
  spatializationAlgorithm=SPATIALIZATION_HRTF
  stereoSpread=2000
  attenuationShapeExtents {x=800}
 }
 concurrencyOverrides {bLimitToOwner=true, maxCount=4, resolutionRule=StopOldest}
 concurrencySet [/VampireKatanaGameplay/Items/VampireKatana/Sound/Settings/VampireKatana_Dash_GlobalLimit_Conc]
}
