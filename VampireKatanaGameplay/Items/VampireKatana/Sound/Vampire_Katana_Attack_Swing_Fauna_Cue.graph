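# Sound cue for the Vampire Katana's "Fauna" attack swing, set up as a melee weapon sound with close-range katana attenuation and player melee concurrency.
# Each branch layers random pitched-up swing and power waves with pitched-down, delayed second-swing and power waves.
# A quality-level node sends its first two levels to four-variation pools with staggered delays, and its last two to two-variation pools, excluded from branch culling, sharing one delay.
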SoundCue /VampireKatanaGameplay/Items/VampireKatana/Sound/Vampire_Katana_Attack_Swing_Fauna_Cue
root=QualityLevel_0
nodes [
 {id=Delay_0, children [Random_2], properties {delayMax=0.37, delayMin=0.37}}
 {id=Delay_1, children [Random_1], properties {delayMax=0.28, delayMin=0.28}}
 {id=Delay_2, children [Mixer_8], properties {delayMax=0.4, delayMin=0.4}}
 {id=Mixer_1, children [Mixer_4,Modulator_1], properties {inputVolume [0.7,1]}}
 {id=Mixer_4, children [Modulator_0,Random_0], properties {inputVolume [0.9,0.7]}}
 {id=Mixer_5, children [Delay_0,Delay_1], properties {inputVolume [1,0.7]}}
 {id=Mixer_6, children [Mixer_7,Modulator_7], properties {inputVolume [0.7,1]}}
 {id=Mixer_7, children [Modulator_6,Random_8], properties {inputVolume [0.9,0.7]}}
 {id=Mixer_8, children [Random_6,Random_9], properties {inputVolume [1,0.8]}}
 {id=Modulator_0, children [Random_3], properties {pitchMax=1.2, pitchMin=1.1}}
 {id=Modulator_1, children [Mixer_5], properties {pitchMax=0.95, pitchMin=0.85}}
 {id=Modulator_14, children [Mixer_1]}
 {id=Modulator_5, children [Mixer_6]}
 {id=Modulator_6, children [Random_7], properties {pitchMax=1.2, pitchMin=1.1}}
 {id=Modulator_7, children [Delay_2], properties {pitchMax=0.95, pitchMin=0.85}}
 {id=QualityLevel_0, children [Modulator_14,Modulator_14,Modulator_5,Modulator_5]}
 {id=Random_0, children [WavePlayer_16, WavePlayer_17, WavePlayer_20, WavePlayer_21], properties {weights [1,1,1,1]}}
 {id=Random_1, children [WavePlayer_34, WavePlayer_35, WavePlayer_38, WavePlayer_39], properties {weights [1,1,1,1]}}
 {id=Random_2, children [WavePlayer_26, WavePlayer_27, WavePlayer_30, WavePlayer_31], properties {weights [1,1,1,1]}}
 {id=Random_3, children [WavePlayer_4,WavePlayer_5,WavePlayer_6,WavePlayer_7], properties {weights [1,1,1,1]}}
 {id=Random_6, children [WavePlayer_44, WavePlayer_45], properties {bShouldExcludeFromBranchCulling=true, weights [1,1]}}
 {id=Random_7, children [WavePlayer_40, WavePlayer_41], properties {bShouldExcludeFromBranchCulling=true, weights [1,1]}}
 {id=Random_8, children [WavePlayer_42, WavePlayer_43], properties {bShouldExcludeFromBranchCulling=true, weights [1,1]}}
 {id=Random_9, children [WavePlayer_46, WavePlayer_47], properties {bShouldExcludeFromBranchCulling=true, weights [1,1]}}
 {id=WavePlayer_16, properties {soundWaveAssetPtr=&/Vampire_Katana_Swing_Power_C_02}}
 {id=WavePlayer_17, properties {soundWaveAssetPtr=&/Vampire_Katana_Swing_Power_C_03}}
 {id=WavePlayer_20, properties {soundWaveAssetPtr=&/Vampire_Katana_Swing_Power_A_01}}
 {id=WavePlayer_21, properties {soundWaveAssetPtr=&/Vampire_Katana_Swing_Power_A_02}}
 {id=WavePlayer_26, properties {soundWaveAssetPtr=&/Vampire_Katana_Attack_Swing_Second_A_01}}
 {id=WavePlayer_27, properties {soundWaveAssetPtr=&/Vampire_Katana_Attack_Swing_Second_A_02}}
 {id=WavePlayer_30, properties {soundWaveAssetPtr=&/Vampire_Katana_Attack_Swing_Second_A_03}}
 {id=WavePlayer_31, properties {soundWaveAssetPtr=&/Vampire_Katana_Attack_Swing_Second_A_04}}
 {id=WavePlayer_34, properties {soundWaveAssetPtr=&/Vampire_Katana_Swing_Power_D_01}}
 {id=WavePlayer_35, properties {soundWaveAssetPtr=&/Vampire_Katana_Swing_Power_D_02}}
 {id=WavePlayer_38, properties {soundWaveAssetPtr=&/Vampire_Katana_Swing_Power_D_03}}
 {id=WavePlayer_39, properties {soundWaveAssetPtr=&/Vampire_Katana_Swing_Power_D_04}}
 {id=WavePlayer_4, properties {soundWaveAssetPtr=&/Vampire_Katana_Attack_Swing_01}}
 {id=WavePlayer_40, properties {soundWaveAssetPtr=&Vampire_Katana_Attack_Swing_01}}
 {id=WavePlayer_41, properties {soundWaveAssetPtr=&/Vampire_Katana_Attack_Swing_02}}
 {id=WavePlayer_42, properties {soundWaveAssetPtr=&Vampire_Katana_Swing_Power_C_02}}
 {id=WavePlayer_43, properties {soundWaveAssetPtr=&Vampire_Katana_Swing_Power_C_03}}
 {id=WavePlayer_44, properties {soundWaveAssetPtr=&Vampire_Katana_Attack_Swing_Second_A_01}}
 {id=WavePlayer_45, properties {soundWaveAssetPtr=&Vampire_Katana_Attack_Swing_Second_A_02}}
 {id=WavePlayer_46, properties {soundWaveAssetPtr=&Vampire_Katana_Swing_Power_D_01}}
 {id=WavePlayer_47, properties {soundWaveAssetPtr=&Vampire_Katana_Swing_Power_D_02}}
 {id=WavePlayer_5, properties {soundWaveAssetPtr=&Vampire_Katana_Attack_Swing_02}}
 {id=WavePlayer_6, properties {soundWaveAssetPtr=&/Vampire_Katana_Attack_Swing_03}}
 {id=WavePlayer_7, properties {soundWaveAssetPtr=&/Vampire_Katana_Attack_Swing_04}}
]
properties {
 attenuationSettings=/KatanaGameplay/Items/Katana/Sounds/Fort_Weapon_Melee_Katana_Close_Range_Stereo_Att
 maxDistance=6000
 soundClassObject=&Fort_WeaponMelee_PSM
 volumeMultiplier=0.9
 concurrencySet [&Player_Melee_Weapon_Conc]
}
