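# Sound cue for the Vampire Katana's attack target lock, playing a single wave at a raised pitch and reduced volume.
# The root wave player feeds straight to output, with no randomization or mixing.
# It reuses weapon pickup attenuation and the loot collect sound class.
# A concurrency override allows only one instance at a time, blocking new plays while one is active.
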SoundCue /VampireKatanaGameplay/Items/VampireKatana/Sound/Vampire_Katana_Attack_Target_Lock_Cue
root=WavePlayer_0
nodes [{id=WavePlayer_0, properties {soundWaveAssetPtr=&/Vampire_Katana_Attack_Target_Lock_01}}]
properties {
 attenuationSettings=&Fort_Weapon_Pickup_Att
 bOverrideConcurrency=true
 maxDistance=3500
 pitchMultiplier=1.2
 soundClassObject=&Fort_Loot_Collect
 volumeMultiplier=0.219
 concurrencyOverrides {maxCount=1, resolutionRule=PreventNew}
}
