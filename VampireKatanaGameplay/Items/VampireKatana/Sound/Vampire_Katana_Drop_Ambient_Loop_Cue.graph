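# Sound cue for the ambient loop of a dropped Vampire Katana, going by its name and drop attenuation settings.
# A mixer layers a looping idle sound with slightly quieter voice clips.
# The voice layer repeatedly waits a random few seconds, then plays one of six equally weighted clips at lowered pitch.
# Drop attenuation with occlusion and a concurrency override limit its range and simultaneous instances.
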SoundCue /VampireKatanaGameplay/Items/VampireKatana/Sound/Vampire_Katana_Drop_Ambient_Loop_Cue
root=Mixer_0
nodes [
 {id=Delay_6, children [Random_7], properties {delayMax=5, delayMin=3}}
 {id=Looping_0, children [Delay_6]}
 {id=Mixer_0, children [Modulator_0,Modulator_5], properties {inputVolume [1,0.8]}}
 {id=Modulator_0, children [WavePlayer_0], properties {pitchMax=1, volumeMax=1, volumeMin=1}}
 {id=Modulator_5, children [Looping_0], properties {pitchMax=0.8, pitchMin=0.8, volumeMax=1, volumeMin=1}}
 {id=Random_7, children [WavePlayer_1,WavePlayer_2,WavePlayer_3,WavePlayer_4,WavePlayer_6,WavePlayer_7], properties {weights [1,1,1,1,1,1]}}
 {
  id=WavePlayer_0
  properties {bLooping=true, soundWaveAssetPtr=&/Vampire_Katana_Drop_Idle_Loop}
 }
 {id=WavePlayer_1, properties {soundWaveAssetPtr=&/Vampire_Katana_Idle_Voices_06}}
 {id=WavePlayer_2, properties {soundWaveAssetPtr=&/Vampire_Katana_Idle_Voices_05}}
 {id=WavePlayer_3, properties {soundWaveAssetPtr=&/Vampire_Katana_Idle_Voices_04}}
 {id=WavePlayer_4, properties {soundWaveAssetPtr=&/Vampire_Katana_Idle_Voices_03}}
 {id=WavePlayer_6, properties {soundWaveAssetPtr=&/Vampire_Katana_Idle_Voices_02}}
 {id=WavePlayer_7, properties {soundWaveAssetPtr=&/Vampire_Katana_Idle_Voices_01}}
]
properties {
 attenuationSettings=&Fort_Weapon_Drop_Att
 bOverrideConcurrency=true
 maxDistance=3000
 soundClassObject=&Fort_Gadgets
 volumeMultiplier=0.7
 attenuationOverrides {
  bEnableOcclusion=true
  falloffDistance=600
  occlusionTraceChannel=ECC_GameTraceChannel11
  occlusionVolumeAttenuation=0.2
  attenuationShapeExtents {x=250}
 }
 concurrencyOverrides {maxCount=3, resolutionRule=StopLowestPriority, voiceStealReleaseTime=2}
}
