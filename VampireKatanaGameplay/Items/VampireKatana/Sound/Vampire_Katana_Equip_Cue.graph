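# Sound cue for equipping the Vampire Katana, layering randomized equip, grab and delayed reach one-shots with recurring pitched-down voices.
# A quality node gives the first two levels a full mix with an enveloped looping idle sound, and the other two a leaner mix with fewer variations and no idle.
# An outer envelope gradually lowers overall volume.
# Playback uses close-prop attenuation, the melee weapon sound class and a two-instance concurrency cap.
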SoundCue /VampireKatanaGameplay/Items/VampireKatana/Sound/Vampire_Katana_Equip_Cue
root=Enveloper_2
nodes [
 {id=Delay_0, children [Random_1]}
 {id=Delay_1, children [Modulator_0], properties {delayMax=0.2, delayMin=0.2}}
 {id=Delay_2, children [Random_4]}
 {id=Delay_3, children [Random_0], properties {delayMax=0.03, delayMin=0.03}}
 {id=Delay_4, children [Modulator_3], properties {delayMax=0.2, delayMin=0.2}}
 {id=Delay_5, children [Random_3], properties {delayMax=0.03, delayMin=0.03}}
 {id=Delay_6, children [Random_7], properties {delayMax=5, delayMin=3}}
 {id=Delay_7, children [Random_9], properties {delayMax=5, delayMin=3}}
 {
  id=Enveloper_2
  children [QualityLevel_0]
  properties {
   pitchCurve {editorCurveData {keys [{time=0}]}}
   volumeCurve {editorCurveData {keys [
      <at 0 1>
      {arriveTangent=-0.0283, interpMode=RCIM_Cubic, leaveTangent=-0.0283, tangentMode=RCTM_User, time=8, value=0.8846}
      {arriveTangent=-0.1053, interpMode=RCIM_Cubic, leaveTangent=-0.1053, tangentMode=RCTM_User, time=15, value=0.5}
     ]}}
  }
 }
 {
  id=Enveloper_3
  children [WavePlayer_29]
  properties {
   pitchCurve {editorCurveData {keys [{time=0}]}}
   volumeCurve {editorCurveData {keys [
      <at 0 1>
      {arriveTangent=-0.0081, interpMode=RCIM_Cubic, leaveTangent=-0.0081, tangentMode=RCTM_User, time=0.7814, value=0.9991}
      {arriveTangent=-0.0518, interpMode=RCIM_Cubic, leaveTangent=-0.0518, tangentMode=RCTM_User, time=10, value=0.2}
     ]}}
  }
 }
 {id=Looping_0, children [Delay_6]}
 {id=Looping_1, children [Delay_7]}
 {id=Mixer_0, children [Modulator_1,Delay_0,Delay_1,Modulator_5,Modulator_8], properties {inputVolume [1.8,1,1.2,1,1.7]}}
 {id=Mixer_1, children [Modulator_2,Delay_2,Delay_4,Modulator_7], properties {inputVolume [1.8,1,1.2,1]}}
 {id=Modulator_0, children [Random_2], properties {pitchMax=1.1, pitchMin=0.9, volumeMax=1.1, volumeMin=0.9}}
 {id=Modulator_1, children [Delay_3], properties {pitchMax=1.1, pitchMin=1}}
 {id=Modulator_2, children [Delay_5], properties {pitchMax=1.1, pitchMin=1}}
 {id=Modulator_3, children [Random_5], properties {pitchMax=1.1, pitchMin=0.9, volumeMax=1.1, volumeMin=0.9}}
 {id=Modulator_5, children [Looping_0], properties {pitchMax=0.8, pitchMin=0.8, volumeMax=1, volumeMin=1}}
 {id=Modulator_7, children [Looping_1], properties {pitchMax=0.8, pitchMin=0.8, volumeMax=1, volumeMin=1}}
 {id=Modulator_8, children [Enveloper_3]}
 {id=QualityLevel_0, children [Mixer_0,Mixer_0,Mixer_1,Mixer_1]}
 {id=Random_0, children [WavePlayer_5, WavePlayer_8, WavePlayer_13], properties {weights [1,1,1]}}
 {id=Random_1, children [WavePlayer_14, WavePlayer_15, WavePlayer_16], properties {weights [1,1,1]}}
 {id=Random_2, children [WavePlayer_19, WavePlayer_20], properties {weights [1,1]}}
 {id=Random_3, children [WavePlayer_21, WavePlayer_22], properties {bShouldExcludeFromBranchCulling=true, weights [1,1]}}
 {id=Random_4, children [WavePlayer_23, WavePlayer_24], properties {bShouldExcludeFromBranchCulling=true, weights [1,1]}}
 {id=Random_5, children [WavePlayer_25], properties {weights [1]}}
 {
  id=Random_7
  children [WavePlayer_1, WavePlayer_2, WavePlayer_3, WavePlayer_4, WavePlayer_6, WavePlayer_7, WavePlayer_10]
  properties {weights [1,1,1,1,1,1,1]}
 }
 {id=Random_9, children [WavePlayer_26, WavePlayer_27, WavePlayer_28], properties {bShouldExcludeFromBranchCulling=true, weights [1,1,1]}}
 {id=WavePlayer_1, properties {soundWaveAssetPtr=&/Vampire_Katana_EquippedVoices_01}}
 {id=WavePlayer_10, properties {soundWaveAssetPtr=&/Vampire_Katana_EquippedVoices_07}}
 {id=WavePlayer_13, properties {soundWaveAssetPtr=&/Vampire_Katana_Equip_03}}
 {id=WavePlayer_14, properties {soundWaveAssetPtr=&/Vampire_Katana_Equip_Grab_01}}
 {id=WavePlayer_15, properties {soundWaveAssetPtr=&/Vampire_Katana_Equip_Grab_02}}
 {id=WavePlayer_16, properties {soundWaveAssetPtr=&/Vampire_Katana_Equip_Grab_03}}
 {id=WavePlayer_19, properties {soundWaveAssetPtr=&/Vampire_Katana_Equip_Reach_03}}
 {id=WavePlayer_2, properties {soundWaveAssetPtr=&/Vampire_Katana_EquippedVoices_02}}
 {id=WavePlayer_20, properties {soundWaveAssetPtr=&/Vampire_Katana_Equip_Reach_04}}
 {id=WavePlayer_21, properties {soundWaveAssetPtr=&/Vampire_Katana_Equip_01}}
 {id=WavePlayer_22, properties {soundWaveAssetPtr=&/Vampire_Katana_Equip_02}}
 {id=WavePlayer_23, properties {soundWaveAssetPtr=&Vampire_Katana_Equip_Grab_01}}
 {id=WavePlayer_24, properties {soundWaveAssetPtr=&Vampire_Katana_Equip_Grab_02}}
 {id=WavePlayer_25, properties {soundWaveAssetPtr=&Vampire_Katana_Equip_Reach_04}}
 {id=WavePlayer_26, properties {soundWaveAssetPtr=&/Vampire_Katana_EquippedVoices_05}}
 {id=WavePlayer_27, properties {soundWaveAssetPtr=&/Vampire_Katana_EquippedVoices_06}}
 {id=WavePlayer_28, properties {soundWaveAssetPtr=&Vampire_Katana_EquippedVoices_07}}
 {id=WavePlayer_29, properties {bLooping=true, soundWaveAssetPtr=&/Vampire_Katana_Idle_Loop}}
 {id=WavePlayer_3, properties {soundWaveAssetPtr=&/Vampire_Katana_EquippedVoices_03}}
 {id=WavePlayer_4, properties {soundWaveAssetPtr=&/Vampire_Katana_EquippedVoices_04}}
 {id=WavePlayer_5, properties {soundWaveAssetPtr=&Vampire_Katana_Equip_01}}
 {id=WavePlayer_6, properties {soundWaveAssetPtr=&Vampire_Katana_EquippedVoices_05}}
 {id=WavePlayer_7, properties {soundWaveAssetPtr=&Vampire_Katana_EquippedVoices_06}}
 {id=WavePlayer_8, properties {soundWaveAssetPtr=&Vampire_Katana_Equip_02}}
]
properties {
 attenuationSettings=/Game/Sounds/Attenuation/Fort_Prop_Close_Att
 bOverrideConcurrency=true
 maxDistance=1800
 soundClassObject=&Fort_WeaponMelee_PSM
 volumeMultiplier=0.45
 concurrencyOverrides {maxCount=2}
}
