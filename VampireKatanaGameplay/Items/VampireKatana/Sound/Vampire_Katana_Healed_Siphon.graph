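# Sound cue for the Vampire Katana's siphon heal feedback, layering a siphon sound with a heal sound.
# After a short fixed delay, a mixer boosts and blends the modulated siphon wave with one of four equally weighted, pitch-varied heal waves.
# It uses loot-collect sound class and prop-collect attenuation, and its concurrency override allows one instance at a time, blocking new plays rather than cutting off the current one.
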SoundCue /VampireKatanaGameplay/Items/VampireKatana/Sound/Vampire_Katana_Healed_Siphon
root=Delay_0
nodes [
 {id=Delay_0, children [Mixer_0], properties {delayMax=0.1, delayMin=0.1}}
 {id=Mixer_0, children [Modulator_0,Modulator_1], properties {inputVolume [1.2,1.5]}}
 {id=Modulator_0, children [WavePlayer_0]}
 {id=Modulator_1, children [Random_1], properties {pitchMax=1, pitchMin=0.9}}
 {id=Random_1, children [WavePlayer_3,WavePlayer_4,WavePlayer_5,WavePlayer_6], properties {weights [1,1,1,1]}}
 {id=WavePlayer_0, properties {soundWaveAssetPtr=&/Vampire_Katana_Siphon_01}}
 {id=WavePlayer_3, properties {soundWaveAssetPtr=&/Vampire_Katana_Heal_01}}
 {id=WavePlayer_4, properties {soundWaveAssetPtr=&/Vampire_Katana_Heal_02}}
 {id=WavePlayer_5, properties {soundWaveAssetPtr=&/Vampire_Katana_Heal_03}}
 {id=WavePlayer_6, properties {soundWaveAssetPtr=&/Vampire_Katana_Heal_04}}
]
properties {
 attenuationSettings=/Game/Sounds/Attenuation/Fort_Prop_Collect_Att
 bOverrideConcurrency=true
 maxDistance=2097152
 soundClassObject=&Fort_Loot_Collect
 volumeMultiplier=2
 concurrencyOverrides {maxCount=1, resolutionRule=PreventNew}
}
