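# Sound cue for stowing the Vampire Katana melee weapon.
# A Random node picks one of three equally weighted stow wave variations and feeds it through a Modulator at the root.
# Playback uses close-prop attenuation, the melee weapon sound class, a reduced volume and a concurrency cap that limits overlapping instances.
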SoundCue /VampireKatanaGameplay/Items/VampireKatana/Sound/Vampire_Katana_Stow_Cue
root=Modulator_0
nodes [
 {id=Modulator_0, children [Random_0]}
 {id=Random_0, children [WavePlayer_2,WavePlayer_4,WavePlayer_5], properties {weights [1,1,1]}}
 {id=WavePlayer_2, properties {soundWaveAssetPtr=&/Vampire_Katana_StowB_01}}
 {id=WavePlayer_4, properties {soundWaveAssetPtr=&/Vampire_Katana_StowB_02}}
 {id=WavePlayer_5, properties {soundWaveAssetPtr=&/Vampire_Katana_StowB_03}}
]
properties {
 attenuationSettings=/Game/Sounds/Attenuation/Fort_Prop_Close_Att
 maxDistance=1800
 soundClassObject=&Fort_WeaponMelee_PSM
 volumeMultiplier=0.4
 concurrencyOverrides {maxCount=2}
}
